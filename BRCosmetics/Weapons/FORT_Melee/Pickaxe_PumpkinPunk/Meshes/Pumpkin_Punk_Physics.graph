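# Physics asset for the Pumpkin Punk two-handed pickaxe cosmetic, giving its handle and dynamic branch bones capsule collision bodies.
# Two branch groups are constrained to the handle: A is a three-body chain, while B has five bodies and forks at its second.
# Swing and twist limits are set to only one or two degrees per joint, with the looser values on the outer B joints.
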PhysicsAsset /BRCosmetics/Weapons/FORT_Melee/Pickaxe_PumpkinPunk/Meshes/Pumpkin_Punk_Physics
bodies {
 Melee_TwoHanded_Handle [{shapeType=Capsule, length=81.588, radius=5.8119, center=<-0.3175 -1.0593 26.6514>, rotation=<rot -0.358 -89.9995 -0.4276>}]
 dyn_BranchesA_01 [{shapeType=Capsule, length=2.9177, radius=1.4532, center=<1.4242 0.259 0.1179>, rotation=<rot -15.1798 -92.4571 -89.3633>}]
 dyn_BranchesA_02 [{shapeType=Capsule, length=2.1345, radius=1.4774, center=<0.809 -0.0174 0.0359>, rotation=<rot 22.4696 -89.9996 95.6416>}]
 dyn_BranchesA_03 [{shapeType=Capsule, length=2.3407, radius=0.7387, center=<0.9601 -0.0601 0.0689>, rotation=<rot 0 91.7437 -83.2465>}]
 dyn_BranchesB_01 [{shapeType=Capsule, length=4.4293, radius=1.8403, center=<2.1284 -0.4979 0.1414>, rotation=<rot 3.6885 -77.7893 -83.7142>}]
 dyn_BranchesB_02 [{shapeType=Capsule, length=4.772, radius=1.2385, center=<2.406 2.2627 -1.0945>, rotation=<rot 18.5471 -59.8616 -59.4241>}]
 dyn_BranchesB_03 [{shapeType=Capsule, length=3.9737, radius=1.4042, center=<1.4275 -0.0344 -0.0686>, rotation=<rot 3.9883 81.2762 -77.5068>}]
 dyn_BranchesB_04 [{shapeType=Capsule, length=4.7409, radius=0.9348, center=<3.3919 0.9639 -0.4207>, rotation=<rot -0.5206 -102.2006 -84.552>}]
 dyn_BranchesB_05 [{shapeType=Capsule, length=3.4186, radius=1.1074, center=<1.8444 -0.0099 -0.4666>, rotation=<rot 0 -107.2314 -72.304>}]
}
constraints [
 {bone1Name=dyn_BranchesA_02, bone2Name=dyn_BranchesA_01, swing1LimitDegrees=1, swing2LimitDegrees=1, twistLimitDegrees=1}
 {bone1Name=dyn_BranchesA_03, bone2Name=dyn_BranchesA_02, swing1LimitDegrees=1, swing2LimitDegrees=1, twistLimitDegrees=1}
 {bone1Name=dyn_BranchesB_02, bone2Name=dyn_BranchesB_01, swing1LimitDegrees=1, swing2LimitDegrees=1, twistLimitDegrees=1}
 {bone1Name=dyn_BranchesB_03, bone2Name=dyn_BranchesB_02, swing1LimitDegrees=2, swing2LimitDegrees=2, twistLimitDegrees=2}
 {bone1Name=dyn_BranchesB_04, bone2Name=dyn_BranchesB_02, swing1LimitDegrees=2, swing2LimitDegrees=2, twistLimitDegrees=2}
 {bone1Name=dyn_BranchesB_05, bone2Name=dyn_BranchesB_04, swing1LimitDegrees=2, swing2LimitDegrees=2, twistLimitDegrees=2}
 {bone1Name=dyn_BranchesA_01, bone2Name=Melee_TwoHanded_Handle, swing1LimitDegrees=1, swing2LimitDegrees=1, twistLimitDegrees=1}
 {bone1Name=dyn_BranchesB_01, bone2Name=Melee_TwoHanded_Handle, swing1LimitDegrees=1, swing2LimitDegrees=1, twistLimitDegrees=1}
]
